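; Looping gameplay cue that fades a thermal optic post-process effect in and out for a weapon mod.
; On start it caches the current weapon's scope post-process MIDs and plays a timeline that sets their thermal mask and direction, reversing it on removal.
; It tracks local controllers' view-target changes, toggling significance-based custom depth overrides for viewers of the target pawn.
; The mask renders as zero when no local controller views that pawn.
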(event EventOnLoopingStart(Cascade)
  (params (MyTarget :type Actor%20Object%20Reference) (Parameters :type Gameplay%20Cue%20Parameters%20Structure%20%28by%20ref%29) (ParticleComponents :type Array%20of%20Cascade%20Particle%20System%20Component%20Object%20References) (AudioComponents :type Array%20of%20Audio%20Component%20Object%20References))
  ;[0.1185 1 0] Event On Looping Start
  (Utilities|FlowControl|Sequence
    (:then_0
      ;;[0.42 1 0.3989] Init important variables and find our materials to manipulate.
      (Utilities|IsValid MyTarget
        (:Is%20Valid
          (bind _asfort_player_pawn (Utilities|Casting|CastToFortPlayerPawn (Game|GetInstigator MyTarget)))
          (Variables|Transient|SetTarget%20Pawn _asfort_player_pawn)
          (bind _asfort_weapon_ranged (Utilities|Casting|CastToFortWeaponRanged (Variables|/Script/FortniteGame.FortPawn|GetCurrentWeapon _asfort_player_pawn)))
          (CallFunction|CacheMaterialParameters _asfort_weapon_ranged))))
    (:then_1
      (CallFunction|RegisterwithAllLocalPlayerControllers))
    (:then_2
      (Components|Timeline|SetPlayRate (Variables|GCNL_WM_ThermalOptic_PostProcess|GetThermalFadeTimeline) (Variables|Default|GetFadeInPlayrate))
      ;[0.6139 0 0.639] Update thermal percent (0 == no thermal, 1 == full thermal)
      (bind (_direction _thermalramp) (Timeline|ThermalFadeTimeline Play)
        (:Update
          (CallFunction|UpdateMaterials _thermalramp _direction))))))

(event EventOnRemoval(Cascade)
  (params (MyTarget :type Actor%20Object%20Reference) (Parameters :type Gameplay%20Cue%20Parameters%20Structure%20%28by%20ref%29) (ParticleComponents :type Array%20of%20Cascade%20Particle%20System%20Component%20Object%20References) (AudioComponents :type Array%20of%20Audio%20Component%20Object%20References) (BurstCameraShakeInstance :type Camera%20Shake%20Base%20Object%20Reference) (BurstDecalInstance :type Decal%20Actor%20Object%20Reference))
  ;[0.4792 0.2841 0.09471] Event On Looping Removal
  (CallFunction|UnregisterwithAllLocalPlayerControllers)
  (Components|Timeline|SetPlayRate (Variables|GCNL_WM_ThermalOptic_PostProcess|GetThermalFadeTimeline) (Variables|Default|GetFadeOutPlayrate))
  ;[0.6139 0 0.639] Update thermal percent (0 == no thermal, 1 == full thermal)
  (Timeline|ThermalFadeTimeline Reverse))

(event Custom|RegisterWithAllLocalPlayerControllers
  ;[0.45 1 0.5169] RegisterAllLocalPlayerControllers
  (for _array_element (Fortnite|GetLocalFortPlayerControllers 0 true true)
    (Utilities|IsValid _array_element
      (:Is%20Valid
        (EventDispatchers|BindEventtoOnViewTargetChanged _array_element (EventDispatchers|CreateEvent :!event Custom|OnViewTargetChanged))
        (if (== (Pawn|GetViewTarget _array_element) (Variables|Transient|GetTarget%20Pawn))
          (Rendering|SetIgnoreSignifanceBasedCustomDepthRendering _array_element true 0)
          (Utilities|Array|Add (Variables|Transient|GetLocalPlayerControllersViewingTarget))))))
  (CallFunction|UpdateMaterials (Variables|Transient|GetThermal%20Alpha) (Variables|Transient|GetThermal%20Direction)))

(event Custom|UnregisterWithAllLocalPlayerControllers
  (bind _localplayercontrollersviewingtarget (Variables|Transient|GetLocalPlayerControllersViewingTarget))
  ;[1 0.7917 0.3356] UnregisterAllLocalPlayerControllers
  (Utilities|FlowControl|Sequence
    (:then_0
      (for _array_element (Fortnite|GetLocalFortPlayerControllers 0 true true)
        (Utilities|IsValid _array_element
          (:Is%20Valid
            (EventDispatchers|UnbindEventfromOnViewTargetChanged _array_element (EventDispatchers|CreateEvent :!event Custom|OnViewTargetChanged))))))
    (:then_1
      (for _array_element_2 _localplayercontrollersviewingtarget
        (Utilities|IsValid _array_element_2
          (:Is%20Valid
            (Rendering|SetIgnoreSignifanceBasedCustomDepthRendering _array_element_2 false 0)))))
    (:then_2
      (Utilities|Array|Clear _localplayercontrollersviewingtarget))
    (:then_3
      (CallFunction|UpdateMaterials (Variables|Transient|GetThermal%20Alpha) (Variables|Transient|GetThermal%20Direction)))))

(event Custom|OnViewTargetChanged
  (params (PlayerController :type Fort%20Player%20Controller%20Object%20Reference) (OldViewTarget :type Actor%20Object%20Reference) (NewViewTarget :type Actor%20Object%20Reference))
  (bind _localplayercontrollersviewingtarget (Variables|Transient|GetLocalPlayerControllersViewingTarget))
  (bind _target_pawn (Variables|Transient|GetTarget%20Pawn))
  ;[0 1 0.7571] OnSpectatingTargetChanged()
  ; Add or remove this player to LocalPlayerControllersViewingTarget.
  (Utilities|IsValid PlayerController
    (:Is%20Valid
      (if (== OldViewTarget _target_pawn)
        ;;[0.4115 0.301 0.07395] This player is no longer viewing our target pawn, remove them.
        (bind _returnvalue (Utilities|Array|RemoveItem _localplayercontrollersviewingtarget PlayerController))
        (if _returnvalue
          (Rendering|SetIgnoreSignifanceBasedCustomDepthRendering PlayerController false 0))
        (elif (== _target_pawn NewViewTarget)
          ;;[0.2257 0.2706 1] This player is now viewing our target pawn, add them.
          (bind _returnvalue_2 (Utilities|Array|AddUnique _localplayercontrollersviewingtarget PlayerController))
          (if (!= _returnvalue_2 -1)
            (Rendering|SetIgnoreSignifanceBasedCustomDepthRendering PlayerController true 0))))
      (CallFunction|UpdateMaterials (Variables|Transient|GetThermal%20Alpha) (Variables|Transient|GetThermal%20Direction)))))

(fn CacheMaterialParameters
  (params (Target%20Weapon :type Fort%20Weapon%20Ranged%20Object%20Reference))
  (Utilities|Array|Clear (Variables|Transient|GetThermalToggleMIDs))
  (Utilities|IsValid Target%20Weapon
    (:Is%20Valid
      (bind _returnvalue (Scope|GetorCreateScopePostProcessMaterialInstanceDynamics Target%20Weapon))
      (bind _array_element (Utilities|Array|ForEachLoopwithBreak _returnvalue)
        (:LoopBody
          (bind (_array_element_weight _array_element_object) (Utilities|Struct|BreakWeightedBlendable _array_element))
          (bind _asmaterial_instance_dynamic (Utilities|Casting|CastToMaterialInstanceDynamic _array_element_object))
          (Utilities|Array|Add (Variables|Transient|GetThermalToggleMIDs) _asmaterial_instance_dynamic))))))

(fn UpdateMaterials
  (params (Thermal%20Alpha :type Float%20%28single-precision%29) (Direction :type ETimelineDirection%20Enum :default "Forward"))
  (bind _output_get (Variables|Transient|SetThermal%20Alpha Thermal%20Alpha))
  (Variables|Transient|SetThermal%20Direction Direction)
  (Utilities|FlowControl|Sequence
    (:then_0
      ;[1 0.9365 0.2371] Set Thermal Alpha to Render.
      ; If we a local controller is viewing our target pawn then use the current Thermal Alpha
      ; Else, use 0 to hide the effect.
      (bind _array_element (Utilities|Array|ForEachLoopwithBreak (Fortnite|GetLocalFortPlayerControllers 0 true true))
        (:LoopBody
          (Utilities|IsValid _array_element
            (:Is%20Valid
              (if (== (Pawn|GetViewTarget _array_element) (Variables|Transient|GetTarget%20Pawn))
                (LocalVariables|SetThermal%20Alpha%20To%20Render _output_get)))))))
    (:then_1
      ;[0.3646 0.1309 1] Update material parameters.
      (for _array_element_2 (Variables|Transient|GetThermalToggleMIDs)
        (Utilities|IsValid _array_element_2
          (:Is%20Valid
            (Rendering|Material|SetScalarParameterValue _array_element_2 "ToggleThermalMask" (LocalVariables|GetThermal%20Alpha%20To%20Render) :!class /Script/Engine.MaterialInstanceDynamic)
            (Rendering|Material|SetScalarParameterValue _array_element_2 "ToggleThermalDirection" (Math|Conversions|ToFloat(Boolean) (Utilities|Enum|Equal(Enum) Direction "Forward")) :!class /Script/Engine.MaterialInstanceDynamic)))))))

(fn ConstructionScript)
